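# Editor utility list entry widget that shows one asset path, located in the TextureDownscale editor tool.
# When its list item is set, it casts to EUB_PayloadObject_AssetPath, writes AssetPath into the title text and clears the selected look.
# Selection changes and entry release route through SetSelectedState, which toggles a tinted background image between faint and invisible by setting its opacity.
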EditorUtilityWidgetBlueprint /Game/NeverCook/EditorUtilities/TextureDownscale/Assets/EUW_ListItem_AssetPath
graphs [EventGraph {nodes [
   n1 = SetSelectedState!() -> (OutputDelegate, then, IsSelected:bool=~) {is=title K2Node_CustomEvent}
   > SetOpacity!(
    execute
    self:Image Object Reference=Get Image_Selection() -> (Image_Selection:Image Object Reference)
    InOpacity:float=SelectFloat(A=0.2, B=0, bPickA=@n1.IsSelected)
   ) -> (then)
   n2 = Event OnListItemObjectSet()
   > n3 = Cast To EUB_PayloadObject_AssetPath!(execute, Object=@n2.ListItemObject) -> (then, CastFailed, AsEUB Payload Object Asset Path:EUB Payload Object Asset Path Object Reference) {is=title K2Node_DynamicCast}
   > Set Text(
    Text="To Text (String)"(InString=Get AssetPath!(self:EUB Payload Object Asset Path Object Reference=@n3) -> (AssetPath:String))
    self=Get Text_Title() -> (Text_Title:Common Text Object Reference)
   )
   > SetSelectedState!(execute, self, IsSelected:bool=false) -> (then) {is=title K2Node_CallFunction}
   n4 = Event On Item Selection Changed!() -> (OutputDelegate, then, bIsSelected:bool) {is=title K2Node_Event}
   > SetSelectedState(IsSelected=@n4.bIsSelected)
   Event On Entry Released() {is=title K2Node_Event}
   > SetSelectedState(IsSelected=false)
  ]}]
variables []
widgets {
 info {rootWidgetClass=&Overlay, widgetCount=3}
 widgets [
  Overlay_43 {widgetClassPath=&Overlay, widget=Overlay_43}
  Image_Selection {
   widgetClassPath=&Image
   bIsVariable=true
   parent=Overlay_43
   widget=Image_Selection
   properties {
    brush {
     bIsDynamicallyLoaded=false
     drawAs=Image
     imageType=NoImage
     mirroring=NoMirror
     resourceName=None
     resourceObject=null
     tiling=NoTile
     imageSize=<1 1>
     margin=<ltrb 0 0 0 0>
     outlineSettings {
      bUseBrushTransparency=false
      roundingType=HalfHeightRadius
      width=0
      color {colorUseRule=UseColor_Specified, specifiedColor=<rgba 0 0 0 0>}
      cornerRadii=<0 0 0 0>
     }
     tintColor {colorUseRule=UseColor_Specified, specifiedColor=<rgba 1 1 1 1>}
     uVRegion {bIsValid=false, max=<0 0>, min=<0 0>}
    }
    colorAndOpacity=<rgba 0 0.1 0.1 0>
   }
   slot {class=OverlaySlot, properties {horizontalAlignment=HAlign_Fill, verticalAlignment=VAlign_Fill}}
  }
  Text_Title {
   widgetClassPath=&CommonTextBlock
   bIsVariable=true
   parent=Overlay_43
   widget=Text_Title
   properties {
    bIsScrollingEnabled=false
    scrollOrientation=Orient_Vertical
    scrollStyle=&TextScrollStyle-Base_C
    text=List Item
    font {
     bForceMonospaced=false
     bMaterialIsStencil=false
     fontMaterial=null
     fontObject=&NotoSans
     letterSpacing=0
     monospacedWidth=1
     size=9
     skewAmount=0
     typefaceFontName=NotoSans-Regular
     outlineSettings {
      bApplyOutlineToDropShadows=true
      bMiteredCorners=false
      bSeparateFillAlpha=true
      outlineMaterial=&M-FontShadow
      outlineSize=3
      outlineColor=<rgba 0 0 0 0.5>
     }
    }
    shadowOffset=<0 0>
   }
   slot {class=OverlaySlot, properties {horizontalAlignment=HAlign_Fill, verticalAlignment=VAlign_Fill}}
  }
 ]
}
