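# Editor utility widget for bulk-setting per-platform Downscale on large textures.
# Scan uses embedded Python to list textures in a folder meeting minimum dimensions via Asset Registry tags, without loading them.
# Filter loads listed assets, keeping those differing from parsed "Platform:Value" pairs.
# Apply replaces selected assets' Downscale with those pairs, saves them and delists them; a short delay follows showing the loading overlay before work starts.
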EditorUtilityWidgetBlueprint /Game/NeverCook/EditorUtilities/TextureDownscale/EUW_TextureDownscale
graphs [
 EventGraph {nodes [
   "On Clicked (Button_Scan)"() {is=title K2Node_ComponentBoundEvent}
   > Toggle Loading!(exec, IsVisible:bool=true) -> (then:Exec) {is=title K2Node_MacroInstance}
   n1 = Get Input_AssetType() -> (Input_AssetType:Editor Utility Editable Text Box Object Reference)
   n2 = Get Text(self=Get Input_Dimensions_Y() -> (Input_Dimensions_Y:Editor Utility Editable Text Box Object Reference))
   n3 = Get Text(self=Get Input_Dimensions_X() -> (Input_Dimensions_X:Editor Utility Editable Text Box Object Reference))
   n4 = Get Input_AssetsPath() -> (Input_AssetsPath:Editor Utility Editable Text Box Object Reference)
   > n5 = Execute Python Script!(
    execute
    PythonScript="import unreal\r\n\r\nasset_registry = unreal.AssetRegistryHelpers.get_asset_registry()\r\n\r\nassets = asset_registry.get_assets_by_path(texture_folder, recursive=True)\r\ntextures=[]\r\nfor asset_data in assets:\r\n    if asset_data.asset_class_path.asset_name == asset_type:\r\n        ##print(asset_data.package_name)\r\n        dimensions = asset_data.get_tag_value(\"Dimensions\")\r\n        if dimensions is None:\r\n            print(\"Invalid Dimensions\")\r\n            continue\r\n        width, height = map(int, dimensions.split(\"x\"))\r\n        if width < filter_width or height < filter_height:\r\n            ##print(\"Skip dimensions: \", dimensions)\r\n            continue\r\n        textures.append(asset_data.package_name)"
    texture_folder:String="To String (Text)"(InText=Get Text(self=@n4))
    asset_type:String="To String (Text)"(InText=Get Text(self=@n1))
    filter_width:int=String To Integer(InString="To String (Text)"(InText=@n3))
    filter_height:int=String To Integer(InString="To String (Text)"(InText=@n2))
   ) -> (then, ReturnValue, textures:Array of Strings) {inputs [texture_folder, asset_type, filter_width, filter_height], outputs [textures]}
   > Branch(Condition=@n5)
   > n6 = Set ScanResult(ScanResult:Array of Strings=@n5.textures) -> (Output_Get:Array of Strings)
   > RefreshList!(execute, self, Items:Array of Strings=@n6) -> (then)
   > Toggle Loading()
   "On Clicked (Button_SelectAll)"() {is=title K2Node_ComponentBoundEvent}
   > n7 = For Each Loop(Array:Array of Object References=GetListItems!(self:List View Object Reference=Get EditorUtilityListView_575() -> (EditorUtilityListView_575:Editor Utility List View Object Reference)) -> (ReturnValue:Array of Object References)) -> (Array Element:Object Reference)
   > Set Item Selection!(execute, self:List View Object Reference=Get EditorUtilityListView_575(), Item:Object Reference=@n7, bSelected:bool=true) -> (then) {is=title K2Node_CallFunction}
   "On Clicked (Button_Apply)"() {is=title K2Node_ComponentBoundEvent}
   > Toggle Loading(IsVisible=true)
   > Clear(TargetArray=Get SelectedAssetPaths() -> (SelectedAssetPaths:Array of Strings))
   > n8 = GetSelectedItems!(execute, self:List View Object Reference=Get EditorUtilityListView_575()) -> (then, Items:Array of Object References, ReturnValue:bool)
   > n9 = For Each Loop(Array=@n8)
   n10 = Get Input_DownscaleValues() -> (Input_DownscaleValues:Editor Utility Multi Line Editable Text Box Object Reference)
   n11 = ParsePlatforms!(execute=@n9.Completed, self, InputString:String="To String (Text)"(InText=Get Text(self:"Text Box (Multi- Line) Object Reference"=@n10))) -> (then, Result:"Map of Strings to Float (double-precision)s")
   > Execute Python Script!(
    execute
    PythonScript="import unreal\r\n\r\nasset_registry = unreal.AssetRegistryHelpers.get_asset_registry()\r\n\r\nfor asset_path in asset_paths:\r\n    asset_data = asset_registry.get_assets_by_package_name(asset_path)\r\n    if asset_data:\r\n        asset = asset_data[0].get_asset() \r\n        if asset is None:\r\n            print(\"Not found: \", asset_path)\r\n            continue\r\n\r\n        new_value = unreal.PerPlatformFloat()\r\n        per_platform = new_value.get_editor_property(\"per_platform\")\r\n        for platform, scale in filter_downscale.items():\r\n            per_platform[platform] = scale\r\n        asset.set_editor_property(\"downscale\", new_value)\r\n        unreal.EditorAssetLibrary.save_loaded_asset(asset)"
    asset_paths:Array of Strings=Get SelectedAssetPaths()
    filter_downscale:"Map of Strings to Float (double-precision)s"=@n11
   ) -> (then, ReturnValue) {inputs [asset_paths, filter_downscale]}
   > Clear Selection(self:List View Object Reference=Get EditorUtilityListView_575())
   > Toggle Loading()
   n12 = Cast To EUB_PayloadObject_AssetPath!(execute=@n9.LoopBody, Object=@n9) -> (then, CastFailed, AsEUB Payload Object Asset Path:EUB Payload Object Asset Path Object Reference) {is=title K2Node_DynamicCast}
   > Add(TargetArray=Get SelectedAssetPaths(), NewItem=Get AssetPath!(self:EUB Payload Object Asset Path Object Reference=@n12) -> (AssetPath:String)) {is=title K2Node_CallArrayFunction}
   > RemoveItem!(execute, self:List View Object Reference=Get EditorUtilityListView_575(), Item:Object Reference=@n12) -> (then)
   "On Clicked (Button_DeselectAll)"() {is=title K2Node_ComponentBoundEvent}
   n13 = Get EditorUtilityListView_575()
   > Set Selected Item!(execute, self:List View Object Reference=@n13, Item:Object Reference) -> (then) {is=title K2Node_CallFunction}
   > Clear Selection(self=@n13)
   "On Clicked (Button_Browse)"() {is=title K2Node_ComponentBoundEvent}
   > n14 = Cast To EUB_PayloadObject_AssetPath(Object=Get Selected Item!(self:List View Object Reference=Get EditorUtilityListView_575()) -> (ReturnValue:Object Reference) {is=title K2Node_CallFunction})
   > Execute Python Script!(execute, PythonScript="import unreal\r\n\r\nunreal.EditorAssetLibrary.sync_browser_to_objects([path])", path:String=Get AssetPath(self=@n14)) -> (then, ReturnValue) {inputs [path]}
   "On Clicked (Button_Filter)"() {is=title K2Node_ComponentBoundEvent}
   > Toggle Loading(IsVisible=true)
   n15 = Get Input_DownscaleValues()
   > n16 = ParsePlatforms(InputString="To String (Text)"(InText=Get Text(self=@n15)))
   > n17 = Execute Python Script!(
    execute
    PythonScript="import unreal\r\n\r\nasset_registry = unreal.AssetRegistryHelpers.get_asset_registry()\r\n\r\nprint(filter_downscale)\r\n\r\nfiltered_paths = []\r\nfor asset_path in asset_paths:\r\n    asset_data = asset_registry.get_assets_by_package_name(asset_path)\r\n    if asset_data:\r\n        asset = asset_data[0].get_asset() \r\n        if asset is None:\r\n            print(\"Not found: \", asset_path)\r\n            continue\r\n        \r\n        prop_downscale = asset.get_editor_property(\"downscale\")\r\n        prop_per_platforn = prop_downscale.get_editor_property(\"per_platform\")\r\n  \r\n        found_inconsistency = False\r\n        for platform, scale in filter_downscale.items():\r\n            if platform not in prop_per_platforn:\r\n                found_inconsistency = True\r\n                break\r\n            elif prop_per_platforn[platform] != scale:\r\n                found_inconsistency = True\r\n                break\r\n        \r\n        if found_inconsistency:\r\n            filtered_paths.append(asset_path)"
    asset_paths:Array of Strings=Get ScanResult() -> (ScanResult:Array of Strings)
    filter_downscale:"Map of Strings to Float (double-precision)s"=@n16
   ) -> (then, ReturnValue, filtered_paths:Array of Strings) {inputs [asset_paths, filter_downscale], outputs [filtered_paths]}
   > n18 = Set ScanResult(ScanResult=@n17.filtered_paths)
   > RefreshList(Items=@n18)
   > Toggle Loading()
   "On Item Selection Changed (EditorUtilityListView_575)"!() -> (OutputDelegate, then, Item:Object Reference, bIsSelected:bool) {is=title K2Node_ComponentBoundEvent}
   > n19 = GetSelectedItems(self=Get EditorUtilityListView_575())
   > Set bIsEnabled!(
    execute
    bIsEnabled:bool=@n19.ReturnValue
    self:Widget Object Reference=Get Button_Apply() -> (Button_Apply:Editor Utility Button Object Reference)
   ) -> (then, Output_Get:bool)
   n20 = RefreshList!() -> (OutputDelegate, then, Items:Array of Strings) {is=title K2Node_CustomEvent}
   > Clear(TargetArray:Array of EUB Payload Object Asset Path Object References=Get ListItems() -> (ListItems:Array of EUB Payload Object Asset Path Object References))
   > n21 = For Each Loop(Array:Array of Strings=@n20.Items) -> (Array Element:String)
   Set List Items!(
    execute=@n21.Completed
    self:List View Object Reference=Get EditorUtilityListView_575()
    InListItems:Array of Object References=Get ListItems()
   ) -> (then) {is=title K2Node_CallFunction}
   n22 = Create Editor Utility Object!(
    execute=@n21.LoopBody
    Owner:Object Reference=Self-Reference()
    UtilityType:Editor Utility Object Class Reference=/Game/NeverCook/EditorUtilities/TextureDownscale/Assets/EUB_PayloadObject_AssetPath.EUB_PayloadObject_AssetPath_C
   ) -> (then, ReturnValue:Structure)
   > n23 = Get Object from Handle!(execute, Handle:"Structure (by ref)"=@n22) -> (then, ReturnValue:Object Reference) {is=title K2Node_CallFunction}
   > n24 = Cast To EUB_PayloadObject_AssetPath(Object=@n23)
   > Set AssetPath!(execute, AssetPath:String=@n21, self:EUB Payload Object Asset Path Object Reference=@n24) -> (then, Output_Get:String)
   > Add(
    TargetArray:Array of EUB Payload Object Asset Path Object References=Get ListItems()
    NewItem:"EUB Payload Object Asset Path Object Reference (by ref)"=@n24
   )
  ]}
 ParsePlatforms {nodes [
   n1 = ParsePlatforms!() -> (then, InputString:String) {is=title K2Node_FunctionEntry}
   > n2 = Set L PerPlatform(L PerPlatform:Array of Strings=ParseIntoArray(SourceString=@n1, Delimiter=",")) -> (Output_Get:Array of Strings)
   > n3 = For Each Loop(Array=@n2)
   Return Node!(
    execute=@n3.Completed
    Result:"Map of Strings to Float (double-precision)s"=Get L Result() -> (L Result:"Map of Strings to Float (double-precision)s")
   )
   n4 = Split(SourceString=@n3, InStr=:, SearchDir=FromStart)
   > Branch(Condition=@n4.ReturnValue)
   > Add!(
    execute
    TargetMap:"Map of Strings to Float (double-precision)s"=Get L Result()
    Key:"String (by ref)"=Trim!(SourceString:String=@n4) -> (ReturnValue:String)
    Value:"Float (double-precision) (by ref)"=String To Float(InString=Trim(SourceString=@n4.RightS))
   ) -> (then) {is=title K2Node_CallFunction}
  ]}
 ToggleLoading {nodes [
   n1 = Inputs!() -> (exec:Exec, IsVisible:bool=~)
   > n2 = Branch(Condition=@n1)
   n3 = Get Border_Loading() -> (Border_Loading:Border Object Reference)
   > Set Visibility(Visibility=Visible, self=@n3)
   > n4 = Delay!(execute, WorldContextObject:Object Reference=Self-Reference(), Duration=0.1) -> (then)
   Outputs!(then:Exec=[@n4.then, @n5.then])
   n5 = Set Visibility(execute=@n2.else, Visibility=Collapsed, self=@n3)
  ]}
]
variables [
 ListItems {type="array<EUB_PayloadObject_AssetPath_C>", default []}
 ScanResult {type="array<string>", default []}
 SelectedAssetPaths {type="array<string>", default []}
]
widgets {
 info {rootWidgetClass=&Overlay, widgetCount=30}
 widgets [
  Overlay_0 {widgetClassPath=&Overlay, widget=Overlay_0}
  StackBox_1 {
   widgetClassPath=&StackBox
   parent=Overlay_0
   widget=StackBox_1
   properties {bOverride_Cursor=true, orientation=Orient_Vertical, visibility=Visible}
   slot {class=OverlaySlot, properties {horizontalAlignment=HAlign_Fill, verticalAlignment=VAlign_Fill}}
  }
  GridPanel_166 {widgetClassPath=&GridPanel, parent=StackBox_1, widget=GridPanel_166, properties {columnFill [0.5,0.5]}, slot {class=StackBoxSlot}}
  CommonTextBlock_1 {
   widgetClassPath=&CommonTextBlock
   parent=GridPanel_166
   widget=CommonTextBlock_1
   properties {
    bIsScrollingEnabled=false
    scrollOrientation=Orient_Vertical
    scrollStyle=&TextScrollStyle-Base_C
    text=Folder:
    font {
     bForceMonospaced=false
     bMaterialIsStencil=false
     fontMaterial=null
     fontObject=&NotoSans
     letterSpacing=0
     monospacedWidth=1
     size=9
     skewAmount=0
     typefaceFontName=NotoSans-Regular
     outlineSettings {
      bApplyOutlineToDropShadows=true
      bMiteredCorners=false
      bSeparateFillAlpha=true
      outlineMaterial=&M-FontShadow
      outlineSize=3
      outlineColor=<rgba 0 0 0 0.5>
     }
    }
    shadowOffset=<0 0>
   }
   slot {class=GridSlot}
  }
  Input_AssetsPath {
   widgetClassPath=&EditorUtilityEditableTextBox
   bIsVariable=true
   parent=GridPanel_166
   widget=Input_AssetsPath
   properties {text=/OfferCatalog/Textures/BattleRoyale/BattlePass/}
   slot {class=GridSlot, properties {columnSpan=2, row=1}}
  }
  CommonTextBlock_3 {
   widgetClassPath=&CommonTextBlock
   parent=GridPanel_166
   widget=CommonTextBlock_3
   properties {
    bIsScrollingEnabled=false
    scrollOrientation=Orient_Vertical
    scrollStyle=&TextScrollStyle-Base_C
    text=Asset Type:
    font {
     bForceMonospaced=false
     bMaterialIsStencil=false
     fontMaterial=null
     fontObject=&NotoSans
     letterSpacing=0
     monospacedWidth=1
     size=9
     skewAmount=0
     typefaceFontName=NotoSans-Regular
     outlineSettings {
      bApplyOutlineToDropShadows=true
      bMiteredCorners=false
      bSeparateFillAlpha=true
      outlineMaterial=&M-FontShadow
      outlineSize=3
      outlineColor=<rgba 0 0 0 0.5>
     }
    }
    shadowOffset=<0 0>
   }
   slot {class=GridSlot, properties {row=2}}
  }
  Input_AssetType {
   widgetClassPath=&EditorUtilityEditableTextBox
   bIsVariable=true
   parent=GridPanel_166
   widget=Input_AssetType
   properties {bIsEnabled=false, text=Texture2D}
   slot {class=GridSlot, properties {row=4}}
  }
  CommonTextBlock_2 {
   widgetClassPath=&CommonTextBlock
   parent=GridPanel_166
   widget=CommonTextBlock_2
   properties {
    bIsScrollingEnabled=false
    scrollOrientation=Orient_Vertical
    scrollStyle=&TextScrollStyle-Base_C
    text=Dimensions:
    font {
     bForceMonospaced=false
     bMaterialIsStencil=false
     fontMaterial=null
     fontObject=&NotoSans
     letterSpacing=0
     monospacedWidth=1
     size=9
     skewAmount=0
     typefaceFontName=NotoSans-Regular
     outlineSettings {
      bApplyOutlineToDropShadows=true
      bMiteredCorners=false
      bSeparateFillAlpha=true
      outlineMaterial=&M-FontShadow
      outlineSize=3
      outlineColor=<rgba 0 0 0 0.5>
     }
    }
    shadowOffset=<0 0>
   }
   slot {class=GridSlot, properties {column=1, row=2}}
  }
  StackBox_0 {widgetClassPath=&StackBox, parent=GridPanel_166, widget=StackBox_0, slot {class=GridSlot, properties {column=1, row=4}}}
  Input_Dimensions_X {
   widgetClassPath=&EditorUtilityEditableTextBox
   bIsVariable=true
   parent=StackBox_0
   widget=Input_Dimensions_X
   properties {text=1024}
   slot {class=StackBoxSlot}
  }
  Input_Dimensions_Y {
   widgetClassPath=&EditorUtilityEditableTextBox
   bIsVariable=true
   parent=StackBox_0
   widget=Input_Dimensions_Y
   properties {text=1024}
   slot {class=StackBoxSlot}
  }
  CommonTextBlock_4 {
   widgetClassPath=&CommonTextBlock
   parent=GridPanel_166
   widget=CommonTextBlock_4
   properties {
    bIsScrollingEnabled=false
    scrollOrientation=Orient_Vertical
    scrollStyle=&TextScrollStyle-Base_C
    text=Per Platform Downscale Values:
    font {
     bForceMonospaced=false
     bMaterialIsStencil=false
     fontMaterial=null
     fontObject=&NotoSans
     letterSpacing=0
     monospacedWidth=1
     size=9
     skewAmount=0
     typefaceFontName=NotoSans-Regular
     outlineSettings {
      bApplyOutlineToDropShadows=true
      bMiteredCorners=false
      bSeparateFillAlpha=true
      outlineMaterial=&M-FontShadow
      outlineSize=3
      outlineColor=<rgba 0 0 0 0.5>
     }
    }
    shadowOffset=<0 0>
   }
   slot {class=GridSlot, properties {row=5}}
  }
  Input_DownscaleValues {
   widgetClassPath=/Script/Blutility.EditorUtilityMultiLineEditableTextBox
   bIsVariable=true
   parent=GridPanel_166
   widget=Input_DownscaleValues
   properties {text="Mobile:2, Switch:2"}
   slot {class=GridSlot, properties {columnSpan=2, row=6}}
  }
  StackBox_2 {
   widgetClassPath=&StackBox
   parent=GridPanel_166
   widget=StackBox_2
   slot {class=GridSlot, properties {horizontalAlignment=HAlign_Left, row=7, verticalAlignment=VAlign_Top}}
  }
  Button_Scan {widgetClassPath=&EditorUtilityButton, parent=StackBox_2, widget=Button_Scan, slot {class=StackBoxSlot}}
  CommonTextBlock {
   widgetClassPath=&CommonTextBlock
   parent=Button_Scan
   widget=CommonTextBlock
   properties {
    bIsScrollingEnabled=false
    scrollOrientation=Orient_Vertical
    scrollStyle=&TextScrollStyle-Base_C
    text=Scan Folder
    font {
     bForceMonospaced=false
     bMaterialIsStencil=false
     fontMaterial=null
     fontObject=&NotoSans
     letterSpacing=0
     monospacedWidth=1
     size=9
     skewAmount=0
     typefaceFontName=NotoSans-Regular
     outlineSettings {
      bApplyOutlineToDropShadows=true
      bMiteredCorners=false
      bSeparateFillAlpha=true
      outlineMaterial=&M-FontShadow
      outlineSize=3
      outlineColor=<rgba 0 0 0 0.5>
     }
    }
    shadowOffset=<0 0>
   }
   slot {class=ButtonSlot, properties {horizontalAlignment=HAlign_Fill, verticalAlignment=VAlign_Fill, padding=<ltrb 0 0 0 0>}}
  }
  Button_Filter {
   widgetClassPath=&EditorUtilityButton
   parent=StackBox_2
   widget=Button_Filter
   properties {toolTipText=Filter items that do NOT include these values. This can take a while because it loads each asset to check its actual value.}
   slot {class=StackBoxSlot}
  }
  CommonTextBlock_7 {
   widgetClassPath=&CommonTextBlock
   parent=Button_Filter
   widget=CommonTextBlock_7
   properties {
    bIsScrollingEnabled=false
    scrollOrientation=Orient_Vertical
    scrollStyle=&TextScrollStyle-Base_C
    text=Filter
    font {
     bForceMonospaced=false
     bMaterialIsStencil=false
     fontMaterial=null
     fontObject=&NotoSans
     letterSpacing=0
     monospacedWidth=1
     size=9
     skewAmount=0
     typefaceFontName=NotoSans-Regular
     outlineSettings {
      bApplyOutlineToDropShadows=true
      bMiteredCorners=false
      bSeparateFillAlpha=true
      outlineMaterial=&M-FontShadow
      outlineSize=3
      outlineColor=<rgba 0 0 0 0.5>
     }
    }
    shadowOffset=<0 0>
   }
   slot {class=ButtonSlot, properties {horizontalAlignment=HAlign_Fill, verticalAlignment=VAlign_Fill, padding=<ltrb 0 0 0 0>}}
  }
  Button_SelectAll {widgetClassPath=&EditorUtilityButton, parent=StackBox_2, widget=Button_SelectAll, slot {class=StackBoxSlot}}
  CommonTextBlock_6 {
   widgetClassPath=&CommonTextBlock
   parent=Button_SelectAll
   widget=CommonTextBlock_6
   properties {
    bIsScrollingEnabled=false
    scrollOrientation=Orient_Vertical
    scrollStyle=&TextScrollStyle-Base_C
    text=Select All
    font {
     bForceMonospaced=false
     bMaterialIsStencil=false
     fontMaterial=null
     fontObject=&NotoSans
     letterSpacing=0
     monospacedWidth=1
     size=9
     skewAmount=0
     typefaceFontName=NotoSans-Regular
     outlineSettings {
      bApplyOutlineToDropShadows=true
      bMiteredCorners=false
      bSeparateFillAlpha=true
      outlineMaterial=&M-FontShadow
      outlineSize=3
      outlineColor=<rgba 0 0 0 0.5>
     }
    }
    shadowOffset=<0 0>
   }
   slot {class=ButtonSlot, properties {horizontalAlignment=HAlign_Fill, verticalAlignment=VAlign_Fill, padding=<ltrb 0 0 0 0>}}
  }
  Button_DeselectAll {widgetClassPath=&EditorUtilityButton, parent=StackBox_2, widget=Button_DeselectAll, slot {class=StackBoxSlot}}
  CommonTextBlock_10 {
   widgetClassPath=&CommonTextBlock
   parent=Button_DeselectAll
   widget=CommonTextBlock_10
   properties {
    bIsScrollingEnabled=false
    scrollOrientation=Orient_Vertical
    scrollStyle=&TextScrollStyle-Base_C
    text=Clear Selection
    font {
     bForceMonospaced=false
     bMaterialIsStencil=false
     fontMaterial=null
     fontObject=&NotoSans
     letterSpacing=0
     monospacedWidth=1
     size=9
     skewAmount=0
     typefaceFontName=NotoSans-Regular
     outlineSettings {
      bApplyOutlineToDropShadows=true
      bMiteredCorners=false
      bSeparateFillAlpha=true
      outlineMaterial=&M-FontShadow
      outlineSize=3
      outlineColor=<rgba 0 0 0 0.5>
     }
    }
    shadowOffset=<0 0>
   }
   slot {class=ButtonSlot, properties {horizontalAlignment=HAlign_Fill, verticalAlignment=VAlign_Fill, padding=<ltrb 0 0 0 0>}}
  }
  Button_Browse {widgetClassPath=&EditorUtilityButton, parent=StackBox_2, widget=Button_Browse, slot {class=StackBoxSlot}}
  CommonTextBlock_11 {
   widgetClassPath=&CommonTextBlock
   parent=Button_Browse
   widget=CommonTextBlock_11
   properties {
    bIsScrollingEnabled=false
    scrollOrientation=Orient_Vertical
    scrollStyle=&TextScrollStyle-Base_C
    text=Browse To Asset
    font {
     bForceMonospaced=false
     bMaterialIsStencil=false
     fontMaterial=null
     fontObject=&NotoSans
     letterSpacing=0
     monospacedWidth=1
     size=9
     skewAmount=0
     typefaceFontName=NotoSans-Regular
     outlineSettings {
      bApplyOutlineToDropShadows=true
      bMiteredCorners=false
      bSeparateFillAlpha=true
      outlineMaterial=&M-FontShadow
      outlineSize=3
      outlineColor=<rgba 0 0 0 0.5>
     }
    }
    shadowOffset=<0 0>
   }
   slot {class=ButtonSlot, properties {horizontalAlignment=HAlign_Fill, verticalAlignment=VAlign_Fill, padding=<ltrb 0 0 0 0>}}
  }
  StackBox {
   widgetClassPath=&StackBox
   parent=GridPanel_166
   widget=StackBox
   slot {class=GridSlot, properties {column=1, horizontalAlignment=HAlign_Right, row=7, verticalAlignment=VAlign_Top}}
  }
  Button_Apply {
   widgetClassPath=&EditorUtilityButton
   parent=StackBox
   widget=Button_Apply
   properties {bIsEnabled=false, toolTipText=Requires Selection}
   slot {class=StackBoxSlot}
  }
  CommonTextBlock_8 {
   widgetClassPath=&CommonTextBlock
   parent=Button_Apply
   widget=CommonTextBlock_8
   properties {
    bIsScrollingEnabled=false
    scrollOrientation=Orient_Vertical
    scrollStyle=&TextScrollStyle-Base_C
    text=Apply
    font {
     bForceMonospaced=false
     bMaterialIsStencil=false
     fontMaterial=null
     fontObject=&NotoSans
     letterSpacing=0
     monospacedWidth=1
     size=9
     skewAmount=0
     typefaceFontName=NotoSans-Regular
     outlineSettings {
      bApplyOutlineToDropShadows=true
      bMiteredCorners=false
      bSeparateFillAlpha=true
      outlineMaterial=&M-FontShadow
      outlineSize=3
      outlineColor=<rgba 0 0 0 0.5>
     }
    }
    shadowOffset=<0 0>
   }
   slot {class=ButtonSlot, properties {horizontalAlignment=HAlign_Fill, verticalAlignment=VAlign_Fill, padding=<ltrb 0 0 0 0>}}
  }
  EditorUtilityListView_575 {
   widgetClassPath=/Script/Blutility.EditorUtilityListView
   bIsVariable=true
   parent=StackBox_1
   widget=EditorUtilityListView_575
   properties {
    bClearSelectionOnClick=true
    entryWidgetClass=/Game/NeverCook/EditorUtilities/TextureDownscale/Assets/EUW_ListItem_AssetPath.EUW_ListItem_AssetPath_C
    selectionMode=Multi
   }
   slot {class=StackBoxSlot, properties {size {sizeRule=Fill, value=1}}}
  }
  Border_Loading {
   widgetClassPath=&Border
   bIsVariable=true
   parent=Overlay_0
   widget=Border_Loading
   properties {
    horizontalAlignment=HAlign_Center
    verticalAlignment=VAlign_Center
    visibility=Collapsed
    brushColor=<rgba 0 0 0 0.4>
    padding=<ltrb 0 0 0 20>
   }
   slot {class=OverlaySlot, properties {horizontalAlignment=HAlign_Fill, verticalAlignment=VAlign_Fill}}
  }
  CommonTextBlock_5 {
   widgetClassPath=&CommonTextBlock
   bIsVariable=true
   parent=Border_Loading
   widget=CommonTextBlock_5
   properties {
    bIsScrollingEnabled=false
    scrollOrientation=Orient_Vertical
    scrollStyle=&TextScrollStyle-Base_C
    text=Please Wait. This could take a long time...
    font {
     bForceMonospaced=false
     bMaterialIsStencil=false
     fontMaterial=null
     fontObject=&NotoSans
     letterSpacing=0
     monospacedWidth=1
     size=48
     skewAmount=0
     typefaceFontName=NotoSans-Regular
     outlineSettings {
      bApplyOutlineToDropShadows=true
      bMiteredCorners=false
      bSeparateFillAlpha=true
      outlineMaterial=&M-FontShadow
      outlineSize=3
      outlineColor=<rgba 0 0 0 0.5>
     }
    }
    shadowOffset=<0 0>
   }
   slot {class=BorderSlot, properties {horizontalAlignment=HAlign_Center, verticalAlignment=VAlign_Center, padding=<ltrb 0 0 0 20>}}
  }
 ]
}
